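# Secondary-motion AnimBP for the female Blackwidow body part, layered on a pose copied from the parent mesh.
# AnimDynamics simulates back cloth, web, collar and belt bones, while bone-driven controllers and constraints adjust collar and belt from head and thigh motion.
# EventGraph sets web and belt simulation alphas from pawn speed, adjusting belt for crouch and spine rotation.
# Wrist-to-hand socket distances blend corrective forearm twist transforms.
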AnimBlueprint /BRCosmetics/Characters/Player/Female/Medium/Bodies/F_MED_Blackwidow/Meshes/F_MED_Blackwidow_Body_AnimBP
target_skeleton=&/F_MED_Blackwidow_Body_Skeleton
graphs [
 AnimGraph {nodes [
   n1 = Get web_val() -> (web_val:double)
   Anim Dynamics - Bone: dyn_web_02(Alpha=@n1, ExternalForce="0,0,0") {
    angularSpringConstant=200
    bAngularSpring=true
    bGravityOverrideInSimSpace=true
    bOverrideLinearDamping=true
    gravityScale=0.3
    linearDampingOverride=0.9
    linearSpringConstant=150
    simulationSpace=RootRelative
    alphaBoolBlend {blendOption=Linear}
    alphaScaleBiasClamp {bInterpResult=true, interpSpeedDecreasing=1, interpSpeedIncreasing=1}
    boundBone {boneName=dyn_web_02}
    componentAppliedLinearAccClamp=<3000 3000 3000>
    componentLinearAccScale=<1 1 1>
    componentLinearVelScale=<2 2 2>
    physicsBodyDefinitions [{
      sphereCollisionRadius=0
      boundBone {boneName=dyn_web_02}
      boxExtents=<6 30 30>
      constraintSetup {angularLimitsMax=<3 80 45>, angularLimitsMin=<-3 -80 -4>, angularTarget=<0 0 0>}
      localJointOffset=<15 0 0>
     }]
    relativeSpaceBone {boneName=root}
   }
   n2 = Anim Dynamics - Bone: dyn_cloth_back_01(ComponentPose=Local To Component(LocalPose=Copy Pose From Mesh() {bUseAttachedParent=true}), Alpha=0.45, ExternalForce="0,0,0") {
    angularSpringConstant=150
    bGravityOverrideInSimSpace=true
    bLinearSpring=true
    bOverrideLinearDamping=true
    gravityScale=0.31
    linearDampingOverride=0.985
    linearSpringConstant=150
    alphaBoolBlend {blendOption=Linear}
    boundBone {boneName=dyn_cloth_back_01}
    componentAppliedLinearAccClamp=<50 50 50>
    componentLinearAccScale=<0.1 0.1 0.1>
    componentLinearVelScale=<15 15 15>
    physicsBodyDefinitions [{
      sphereCollisionRadius=0
      boundBone {boneName=dyn_cloth_back_01}
      boxExtents=<30 30 30>
      constraintSetup {
       twistAxis=AxisY
       angularLimitsMax=<0 5 0>
       angularLimitsMin=<0 -5 0>
       angularTarget=<0 0 0>
       linearAxesMax=<2 0.5 1>
       linearAxesMin=<-2 -0.5 -1>
      }
      localJointOffset=<0 0 -15>
     }]
   }
   n3 = Anim Dynamics - Bone: dyn_web_01(ComponentPose=@n2, Alpha=@n1, ExternalForce="0,0,0") {
    angularSpringConstant=300
    bAngularSpring=true
    bGravityOverrideInSimSpace=true
    bOverrideAngularDamping=true
    bOverrideLinearDamping=true
    gravityScale=0.3
    linearDampingOverride=0.9
    linearSpringConstant=150
    simulationSpace=RootRelative
    alphaBoolBlend {blendOption=Linear}
    alphaScaleBiasClamp {bInterpResult=true, interpSpeedDecreasing=1, interpSpeedIncreasing=1}
    boundBone {boneName=dyn_web_01}
    componentAppliedLinearAccClamp=<100 100 100>
    componentLinearAccScale=<1 1 1>
    componentLinearVelScale=<2 2 2>
    physicsBodyDefinitions [{
      sphereCollisionRadius=0
      boundBone {boneName=dyn_web_01}
      boxExtents=<6 30 30>
      constraintSetup {angularLimitsMax=<3 55 70>, angularLimitsMin=<-3 -55 -4>, angularTarget=<0 0 0>}
      localJointOffset=<15 0 0>
     }]
    relativeSpaceBone {boneName=root}
   }
   n4 = Anim Dynamics - Bone: dyn_web_02(ComponentPose=@n3, Alpha=@n1, ExternalForce="0,0,0") {
    angularSpringConstant=100
    bGravityOverrideInSimSpace=true
    bOverrideLinearDamping=true
    gravityScale=0.35
    linearDampingOverride=0.9
    linearSpringConstant=150
    simulationSpace=RootRelative
    alphaBoolBlend {blendOption=Linear}
    alphaScaleBiasClamp {bInterpResult=true, interpSpeedDecreasing=1, interpSpeedIncreasing=1}
    boundBone {boneName=dyn_web_02}
    componentAppliedLinearAccClamp=<1000 1000 1000>
    componentLinearAccScale=<1 1 1>
    componentLinearVelScale=<2 2 2>
    physicsBodyDefinitions [{
      sphereCollisionRadius=0
      boundBone {boneName=dyn_web_02}
      boxExtents=<6 25 25>
      constraintSetup {angularLimitsMax=<3 40 60>, angularLimitsMin=<-3 -40 -4>, angularTarget=<0 0 0>}
      localJointOffset=<10 0 0>
     }]
    relativeSpaceBone {boneName=root}
   }
   n5 = Anim Dynamics - Bone: dyn_web_03(ComponentPose=@n4, Alpha=@n1, ExternalForce="0,0,0") {
    angularDampingOverride=0.985
    angularSpringConstant=100
    bGravityOverrideInSimSpace=true
    bOverrideAngularDamping=true
    gravityScale=0.3
    linearSpringConstant=150
    simulationSpace=RootRelative
    alphaBoolBlend {blendOption=Linear}
    alphaScaleBiasClamp {bInterpResult=true, interpSpeedDecreasing=1, interpSpeedIncreasing=1}
    boundBone {boneName=dyn_web_03}
    componentAppliedLinearAccClamp=<500 500 500>
    componentLinearAccScale=<1 1 1>
    componentLinearVelScale=<8 8 8>
    physicsBodyDefinitions [{
      sphereCollisionRadius=0
      boundBone {boneName=dyn_web_03}
      boxExtents=<6 20 20>
      constraintSetup {angularLimitsMax=<3 55 45>, angularLimitsMin=<-3 -55 -4>, angularTarget=<0 0 0>}
      localJointOffset=<12 0 0>
     }]
    relativeSpaceBone {boneName=root}
   }
   n6 = Anim Dynamics - Bone: dyn_collar_l(ComponentPose=@n5, Alpha=0.45, ExternalForce="0,0,0") {
    angularSpringConstant=150
    bGravityOverrideInSimSpace=true
    bLinearSpring=true
    bOverrideLinearDamping=true
    gravityScale=0.35
    linearDampingOverride=0.985
    linearSpringConstant=150
    alphaBoolBlend {blendOption=Linear}
    boundBone {boneName=dyn_collar_l}
    componentAppliedLinearAccClamp=<100 100 100>
    componentLinearAccScale=<0.1 0.1 0.5>
    componentLinearVelScale=<2 2 2>
    physicsBodyDefinitions [{
      sphereCollisionRadius=0
      boundBone {boneName=dyn_collar_l}
      boxExtents=<30 30 6>
      constraintSetup {twistAxis=AxisY, angularLimitsMax=<0 5 0>, angularLimitsMin=<0 -5 0>, angularTarget=<0 0 0>, linearAxesMax=<3 1 3>, linearAxesMin=<-3 -1 -3>}
      localJointOffset=<0 0 -10>
     }]
   }
   n7 = Anim Dynamics - Bone: dyn_collar_r(ComponentPose=@n6, Alpha=0.45, ExternalForce="0,0,0") {
    angularSpringConstant=150
    bGravityOverrideInSimSpace=true
    bLinearSpring=true
    bOverrideLinearDamping=true
    gravityScale=0.35
    linearDampingOverride=0.985
    linearSpringConstant=150
    alphaBoolBlend {blendOption=Linear}
    boundBone {boneName=dyn_collar_r}
    componentAppliedLinearAccClamp=<100 100 100>
    componentLinearAccScale=<0.1 0.1 0.5>
    componentLinearVelScale=<2 2 2>
    physicsBodyDefinitions [{
      sphereCollisionRadius=0
      boundBone {boneName=dyn_collar_r}
      boxExtents=<30 30 6>
      constraintSetup {twistAxis=AxisY, angularLimitsMax=<0 5 0>, angularLimitsMin=<0 -5 0>, angularTarget=<0 0 0>, linearAxesMax=<3 1 3>, linearAxesMin=<-3 -1 -3>}
      localJointOffset=<0 0 -10>
     }]
   }
   n8 = "Transform (Modify) Bone - Bone: lowerarm_twist_a01_l"(ComponentPose=@n7, Alpha=Get wrist_dist_l() -> (wrist_dist_l:double), Translation="-0.65,-4,0", Rotation="0,-9,0", Scale="1,1,1") {
    rotationMode=BMM_Replace
    rotationSpace=BCS_BoneSpace
    translationMode=BMM_Replace
    translationSpace=BCS_BoneSpace
    alphaBoolBlend {blendOption=Linear}
    boneToModify {boneName=lowerarm_twist_a01_l}
    rotation {yaw=-9}
    translation=<-0.65 -4>
   }
   n9 = "Transform (Modify) Bone - Bone: lowerarm_twist_a01_r"(ComponentPose=@n8, Alpha=Get wrist_dist_r() -> (wrist_dist_r:double), Translation="0.65,4,0", Rotation="0,-9,0", Scale="1,1,1") {
    rotationMode=BMM_Replace
    rotationSpace=BCS_BoneSpace
    translationMode=BMM_Replace
    translationSpace=BCS_BoneSpace
    alphaBoolBlend {blendOption=Linear}
    boneToModify {boneName=lowerarm_twist_a01_r}
    rotation {yaw=-9}
    translation=<0.65 4>
   }
   n10 = "dyn_collar_l.translateZ = curve(head.rotateX) - Bone Driven Controller"!(ComponentPose:Component Space Pose Link Structure=@n9, Alpha:float=1) -> (Pose:Component Space Pose Link Structure) {
    is=title AnimGraphNode_BoneDrivenController
    bAffectTargetTranslationZ=true
    drivingCurve=&/Head_Curve_Z
    sourceComponent=RotationX
    alphaBoolBlend {blendOption=Linear}
    sourceBone {boneName=head}
    targetBone {boneName=dyn_collar_l}
   }
   n11 = "dyn_collar_r.translateZ = curve(head.rotateX) - Bone Driven Controller"!(ComponentPose:Component Space Pose Link Structure=@n10, Alpha:float=1) -> (Pose:Component Space Pose Link Structure) {
    is=title AnimGraphNode_BoneDrivenController
    bAffectTargetTranslationZ=true
    drivingCurve=&/Head_Curve_Z2
    sourceComponent=RotationX
    alphaBoolBlend {blendOption=Linear}
    sourceBone {boneName=head}
    targetBone {boneName=dyn_collar_r}
   }
   n12 = "Constraint - dyn_belt_constraint_l (PPPP)"!(
    ComponentPose=@n11
    Alpha=0.4
    ConstraintWeights_0=0.4656
    ConstraintWeights_1:float=0.4828
    ConstraintWeights_2:float=0
    ConstraintWeights_3:float=0.0516
   ) -> (Pose) {
    alphaBoolBlend {blendOption=Linear}
    boneToModify {boneName=dyn_belt_constraint_l}
    constraintSetup [
     {targetBone {boneName=pelvis}}
     {targetBone {boneName=thigh_twist_01_l}}
     {targetBone {boneName=thigh_twist_01_r}}
     {targetBone {boneName=thigh_l}}
    ]
    constraintWeights [1,1,1,1]
   }
   n13 = "Constraint - dyn_belt_constraint_r (PPP)"!(ComponentPose=@n12, Alpha=1, ConstraintWeights_0=0.03, ConstraintWeights_1:float=0.584, ConstraintWeights_2:float=0.386) -> (Pose) {
    alphaBoolBlend {blendOption=Linear}
    boneToModify {boneName=dyn_belt_constraint_r}
    constraintSetup [{targetBone {boneName=thigh_r}}, {targetBone {boneName=pelvis}}, {targetBone {boneName=thigh_twist_01_r}}]
    constraintWeights [1,1,1]
   }
   n14 = "dyn_belt_constraint_l.translateX = thigh_l.rotateZ * -0.055 - Bone Driven Controller"!(ComponentPose:Component Space Pose Link Structure=@n13, Alpha:float=1) -> (Pose:Component Space Pose Link Structure) {
    is=title AnimGraphNode_BoneDrivenController
    bAffectTargetTranslationX=true
    multiplier=-0.055
    rangeMax=112
    rangeMin=0
    sourceComponent=RotationZ
    alphaBoolBlend {blendOption=Linear}
    sourceBone {boneName=thigh_l}
    targetBone {boneName=dyn_belt_constraint_l}
   }
   n15 = "dyn_belt_driver_r.translateZ = thigh_r.rotateZ * -0.05 - Bone Driven Controller"!(ComponentPose:Component Space Pose Link Structure=@n14, Alpha:float=1) -> (Pose:Component Space Pose Link Structure) {
    is=title AnimGraphNode_BoneDrivenController
    bAffectTargetTranslationZ=true
    multiplier=-0.05
    rangeMax=112
    rangeMin=0
    sourceComponent=RotationZ
    alphaBoolBlend {blendOption=Linear}
    sourceBone {boneName=thigh_r}
    targetBone {boneName=dyn_belt_driver_r}
   }
   n16 = "dyn_belt_driver_f.translateZ = thigh_r.rotateZ * -0.05 - Bone Driven Controller"!(ComponentPose:Component Space Pose Link Structure=@n15, Alpha:float=1) -> (Pose:Component Space Pose Link Structure) {
    is=title AnimGraphNode_BoneDrivenController
    bAffectTargetTranslationZ=true
    multiplier=-0.05
    rangeMax=112
    rangeMin=0
    sourceComponent=RotationZ
    alphaBoolBlend {blendOption=Linear}
    sourceBone {boneName=thigh_r}
    targetBone {boneName=dyn_belt_driver_f}
   }
   n17 = "Transform (Modify) Bone - Bone: dyn_belt_driver_l"(ComponentPose=@n16, Alpha=Get sword_crouch() -> (sword_crouch:double), Translation="0,0.5,1", Rotation="15,-7,-35", Scale="1,1,1") {
    rotationMode=BMM_Additive
    rotationSpace=BCS_ParentBoneSpace
    translationMode=BMM_Additive
    alphaBoolBlend {blendOption=Linear}
    alphaScaleBiasClamp {bInterpResult=true, interpSpeedDecreasing=15, interpSpeedIncreasing=15}
    boneToModify {boneName=dyn_belt_driver_l}
    rotation=<rot 15 -7 -35>
    translation {y=0.5, z=1}
   }
   n18 = Anim Dynamics - Bone: dyn_belt_l(ComponentPose=@n17, Alpha=Get sword_val() -> (sword_val:double), ExternalForce="0,0,0") {
    angularDampingOverride=0.99
    angularSpringConstant=300
    bGravityOverrideInSimSpace=true
    bOverrideAngularDamping=true
    bOverrideLinearDamping=true
    gravityScale=0.3
    linearDampingOverride=0.99
    linearSpringConstant=150
    simulationSpace=RootRelative
    alphaBoolBlend {blendOption=Linear}
    alphaScaleBiasClamp {bInterpResult=true, interpSpeedDecreasing=30, interpSpeedIncreasing=30}
    boundBone {boneName=dyn_belt_l}
    componentLinearAccScale=<0.1 0.1 0.1>
    physicsBodyDefinitions [{
      sphereCollisionRadius=0
      boundBone {boneName=dyn_belt_l}
      boxExtents=<40 40 40>
      constraintSetup {angularLimitsMax=<3 0 70>, angularLimitsMin=<-3 -9 -70>, angularTarget=<0 0 0>}
      localJointOffset=<25 0 -1>
     }]
   }
   n19 = "Transform (Modify) Bone - Bone: dyn_belt_driver_l"(ComponentPose=@n18, Alpha=Get sword_rot() -> (sword_rot:double), Translation="0.35,0,0", Rotation="-12,-23,0", Scale="1,1,1") {
    rotationMode=BMM_Additive
    rotationSpace=BCS_ParentBoneSpace
    translationMode=BMM_Additive
    alphaBoolBlend {blendOption=Linear}
    boneToModify {boneName=dyn_belt_driver_l}
    rotation {pitch=-12, yaw=-23}
    translation {x=0.35}
   }
   Output Pose(Result=Component To Local(ComponentPose=@n19)) {layerGroup=None}
  ]}
 EventGraph {nodes [
   Event BlueprintUpdateAnimation()
   n1 = GetPartSkeletalMeshComponent()
   > Set wrist_dist_l(wrist_dist_l:double=Get Distance Between Two Sockets!(
     Component:Skeletal Mesh Component Object Reference=@n1
     SocketOrBoneNameA:Name=lowerarm_twist_b01_l
     SocketSpaceA:ERelativeTransformSpace Enum=RTS_Component
     SocketOrBoneNameB:Name=middle_metacarpal_l
     SocketSpaceB:ERelativeTransformSpace Enum=RTS_Component
     bRemapRange:bool=true
     InRangeMin:float=10
     InRangeMax:float=7.5
     OutRangeMin:float=0
     OutRangeMax:float=1
    ) -> (ReturnValue:float) {is=title K2Node_CallFunction}) -> (Output_Get:double)
   > Set wrist_dist_r(wrist_dist_r:double=Get Distance Between Two Sockets(
     Component=@n1
     SocketOrBoneNameA=lowerarm_twist_b01_r
     SocketSpaceA=RTS_Component
     SocketOrBoneNameB=middle_metacarpal_r
     SocketSpaceB=RTS_Component
     bRemapRange=true
     InRangeMin=10
     InRangeMax=7.5
     OutRangeMin=0
     OutRangeMax=1
    )) -> (Output_Get:double)
   n2 = MapRangeClamped(Value=Get PawnSpeed2D(), InRangeA=0, InRangeB=100)
   n3 = "float <= float"!(A:double=@n2, B:double=0) -> (ReturnValue:bool) {is=title K2Node_CallFunction}
   > n4 = Branch(Condition=@n3)
   > n5 = Set sword_val(sword_val:double=0.2) -> (Output_Get:double)
   n6 = Branch(execute=[@n13.then, @n5.then], Condition=Get bIsCrouching())
   > Set sword_val(sword_val=0.15)
   > n7 = Set sword_crouch(sword_crouch:double=1) -> (Output_Get:double)
   n8 = Branch(execute=[@n12.then, @n7.then], Condition=@n3)
   > n9 = Set web_val(web_val:double=0.075) -> (Output_Get:double)
   n10 = BreakTransform(InTransform=GetDeltaTransformFromRefPose(self=GetOwnerSkeletalMeshComponent(), BoneName=spine_05, BaseName=pelvis))
   Set sword_rot(execute=[@n9.then, @n11.then], sword_rot:double=MapRangeClamped(Value=BreakRotator(InRot=@n10.Rotation), InRangeA=-8, InRangeB=-18)) -> (Output_Get:double)
   n11 = Set web_val(execute=@n8.else, web_val=float * float(A=@n2, B=0.625))
   n12 = Set sword_crouch(execute=@n6.else, sword_crouch=0)
   n13 = Set sword_val(execute=@n4.else, sword_val=float * float(A=@n2, B=0.5))
   float - float(B=1) {is=title K2Node_CallFunction}
  ]}
]
variables [
 sword_crouch {type=number}
 sword_rot {type=number, default=1}
 sword_val {type=number}
 web_val {type=number, default=0.625}
 wrist_dist_l {type=number}
 wrist_dist_r {type=number}
]
