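# Post-copy Anim Blueprint for the Blackwidow female face accessory part, adding hair physics to the attached parent's copied pose.
# Chained AnimDynamics nodes simulate hair bones with springs and body collision limits, mostly weighted by hair_alpha, then additive bone transforms reposition strands.
# EventGraph turns collar-to-hair socket distances into side-strand rotations and targeting into a right-strand offset.
# Skydive and parachute values are set but unused; some duplicate nodes are disconnected.
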AnimBlueprint /BRCosmetics/Characters/Player/Female/Medium/Bodies/F_MED_Blackwidow/Meshes/F_MED_Blackwidow_FaceAcc_Skeleton_AnimBP
target_skeleton=&/F_MED_Blackwidow_FaceAcc_Skeleton
graphs [
 AnimGraph {nodes [
   n1 = Anim Dynamics - Bone: dyn_hair_r_01(Alpha=1, ExternalForce="0,0,0") {
    angularDampingOverride=0.999
    angularSpringConstant=1000
    bAngularSpring=true
    bGravityOverrideInSimSpace=true
    bOverrideAngularDamping=true
    bOverrideLinearDamping=true
    bUsePlanarLimit=false
    gravityScale=0.45
    linearDampingOverride=0.99
    alphaBoolBlend {blendOption=Linear}
    alphaScaleBiasClamp {interpSpeedDecreasing=2, interpSpeedIncreasing=4}
    boundBone {boneName=dyn_hair_r_01}
    componentAppliedLinearAccClamp=<100 100 100>
    componentLinearVelScale=<50 50 50>
    physicsBodyDefinitions [{
      collisionType=CustomSphere
      sphereCollisionRadius=3
      boundBone {boneName=dyn_hair_r_01}
      boxExtents=<5 30 30>
      constraintSetup {angularLimitsMax=<0 20 17>, angularLimitsMin=<0 -2.5 -8>, angularTarget=<0 0 0>}
      localJointOffset=<-6.5 0 0>
     }]
    sphericalLimits [
     {limitRadius=5, drivingBone {boneName=neck_02}, sphereLocalOffset=<0 0 0>}
     {limitRadius=7, drivingBone {boneName=neck_01}, sphereLocalOffset=<5 10 0>}
    ]
   }
   n2 = Anim Dynamics - Bone: dyn_hair_r_02(ComponentPose=@n1, Alpha=1, ExternalForce="0,0,0") {
    angularDampingOverride=0.999
    bGravityOverrideInSimSpace=true
    bOverrideAngularDamping=true
    bOverrideLinearDamping=true
    gravityScale=0.35
    linearDampingOverride=0.99
    alphaBoolBlend {blendOption=Linear}
    alphaScaleBiasClamp {interpSpeedDecreasing=2, interpSpeedIncreasing=4}
    boundBone {boneName=dyn_hair_r_02}
    componentAppliedLinearAccClamp=<100 100 100>
    componentLinearVelScale=<50 50 50>
    physicsBodyDefinitions [{
      sphereCollisionRadius=0
      boundBone {boneName=dyn_hair_r_02}
      boxExtents=<1 1 1>
      constraintSetup {angularLimitsMax=<0 20 20>, angularLimitsMin=<0 -5 -5>, angularTarget=<0 0 0>}
      localJointOffset=<0 0 0>
     }]
   }
   Anim Dynamics - Bone: dyn_hair_r_03(ComponentPose=@n2, Alpha=1, ExternalForce="0,0,0") {
    angularDampingOverride=0.999
    angularSpringConstant=1000
    bAngularSpring=true
    bGravityOverrideInSimSpace=true
    bOverrideAngularDamping=true
    bOverrideLinearDamping=true
    bUsePlanarLimit=false
    gravityScale=0.45
    linearDampingOverride=0.99
    alphaBoolBlend {blendOption=Linear}
    alphaScaleBiasClamp {interpSpeedDecreasing=2, interpSpeedIncreasing=4}
    boundBone {boneName=dyn_hair_r_03}
    componentAppliedLinearAccClamp=<100 100 100>
    componentLinearVelScale=<50 50 50>
    physicsBodyDefinitions [{
      collisionType=CustomSphere
      sphereCollisionRadius=3
      boundBone {boneName=dyn_hair_r_03}
      boxExtents=<5 30 30>
      constraintSetup {angularLimitsMax=<0 45 5>, angularLimitsMin=<0 -2 -30>, angularTarget=<0 0 0>}
      localJointOffset=<-6.5 0 0>
     }]
    sphericalLimits [{limitRadius=30, drivingBone {boneName=spine_05}, sphereLocalOffset=<5 7 -23.5>}]
   }
   n3 = Get target_hair_r() -> (target_hair_r:double)
   "Transform (Modify) Bone - Bone: dyn_hair_bk_r_02"(Alpha=@n3, Translation="0,0,0", Rotation="25,0,0", Scale="1,1,1") {
    rotationMode=BMM_Additive
    translationMode=BMM_Additive
    alphaBoolBlend {blendOption=Linear}
    alphaScaleBiasClamp {bInterpResult=true, interpSpeedDecreasing=4, interpSpeedIncreasing=6}
    boneToModify {boneName=dyn_hair_bk_r_02}
    rotation {pitch=25}
   }
   n4 = Get hair_alpha() -> (hair_alpha:double)
   n5 = Anim Dynamics - Bone: dyn_hair_bk_01(
    ComponentPose=Local To Component(LocalPose=Copy Pose From Mesh() {bCopyCurves=true, bUseAttachedParent=true})
    Alpha=@n4
    ExternalForce="0,0,0"
   ) {
    angularDampingOverride=0.999
    angularSpringConstant=500
    bAngularSpring=true
    bGravityOverrideInSimSpace=true
    bOverrideAngularDamping=true
    bOverrideLinearDamping=true
    bUsePlanarLimit=false
    bUseSphericalLimits=true
    gravityScale=0.45
    linearDampingOverride=0.99
    alphaBoolBlend {blendOption=Linear}
    alphaScaleBiasClamp {interpSpeedDecreasing=2, interpSpeedIncreasing=4}
    boundBone {boneName=dyn_hair_bk_01}
    componentAppliedLinearAccClamp=<100 100 100>
    componentLinearVelScale=<50 50 50>
    physicsBodyDefinitions [{
      collisionType=CustomSphere
      sphereCollisionRadius=5
      boundBone {boneName=dyn_hair_bk_01}
      boxExtents=<5 30 30>
      constraintSetup {angularLimitsMax=<0 20 7>, angularLimitsMin=<0 -20 -65>, angularTarget=<0 0 0>}
      localJointOffset=<6.5 0 0>
     }]
    sphericalLimits [{limitRadius=7, drivingBone {boneName=neck_02}, sphereLocalOffset=<0 0 0>}]
   }
   n6 = Anim Dynamics - Bone: dyn_hair_bk_02(ComponentPose=@n5, Alpha=@n4, ExternalForce="0,0,0") {
    angularDampingOverride=0.999
    angularSpringConstant=500
    bAngularSpring=true
    bGravityOverrideInSimSpace=true
    bOverrideAngularDamping=true
    bOverrideLinearDamping=true
    bUseSphericalLimits=true
    gravityScale=0.7
    linearDampingOverride=0.99
    alphaBoolBlend {blendOption=Linear}
    alphaScaleBiasClamp {interpSpeedDecreasing=2, interpSpeedIncreasing=4}
    boundBone {boneName=dyn_hair_bk_02}
    componentAppliedLinearAccClamp=<100 100 100>
    componentLinearVelScale=<50 50 50>
    physicsBodyDefinitions [{
      collisionType=CustomSphere
      sphereCollisionRadius=4
      boundBone {boneName=dyn_hair_bk_02}
      boxExtents=<5 40 40>
      constraintSetup {angularLimitsMax=<0 20 7>, angularLimitsMin=<0 -20 -75>, angularTarget=<0 0 0>}
      localJointOffset=<10 0 0>
     }]
    sphericalLimits [{limitRadius=50, drivingBone {boneName=spine_05}, sphereLocalOffset=<-20 25 0>}]
   }
   n7 = Anim Dynamics - Bone: dyn_hair_bk_r_01(ComponentPose=@n6, Alpha=@n4, ExternalForce="0,0,0") {
    angularDampingOverride=0.999
    angularSpringConstant=500
    bAngularSpring=true
    bGravityOverrideInSimSpace=true
    bOverrideAngularDamping=true
    bOverrideLinearDamping=true
    bUsePlanarLimit=false
    gravityScale=0.45
    linearDampingOverride=0.99
    alphaBoolBlend {blendOption=Linear}
    alphaScaleBiasClamp {interpSpeedDecreasing=2, interpSpeedIncreasing=4}
    boundBone {boneName=dyn_hair_bk_r_01}
    componentAppliedLinearAccClamp=<100 100 100>
    componentLinearVelScale=<50 50 50>
    physicsBodyDefinitions [{
      collisionType=CustomSphere
      sphereCollisionRadius=5
      boundBone {boneName=dyn_hair_bk_r_01}
      boxExtents=<5 30 30>
      constraintSetup {angularLimitsMax=<0 20 7>, angularLimitsMin=<0 -20 -65>, angularTarget=<0 0 0>}
      localJointOffset=<6.5 0 0>
     }]
    sphericalLimits [{limitRadius=5, drivingBone {boneName=neck_02}, sphereLocalOffset=<0 0 0>}]
   }
   n8 = Anim Dynamics - Bone: dyn_hair_bk_r_02(ComponentPose=@n7, Alpha=@n4, ExternalForce="0,0,0") {
    angularDampingOverride=0.999
    angularSpringConstant=500
    bAngularSpring=true
    bGravityOverrideInSimSpace=true
    bOverrideAngularDamping=true
    bOverrideLinearDamping=true
    bUseSphericalLimits=true
    gravityScale=0.6
    linearDampingOverride=0.99
    alphaBoolBlend {blendOption=Linear}
    alphaScaleBiasClamp {interpSpeedDecreasing=2, interpSpeedIncreasing=4}
    boundBone {boneName=dyn_hair_bk_r_02}
    componentAppliedLinearAccClamp=<100 100 100>
    componentLinearVelScale=<50 50 50>
    physicsBodyDefinitions [{
      collisionType=CustomSphere
      sphereCollisionRadius=4
      boundBone {boneName=dyn_hair_bk_r_02}
      boxExtents=<5 40 40>
      constraintSetup {angularLimitsMax=<0 20 7>, angularLimitsMin=<0 -20 -75>, angularTarget=<0 0 0>}
      localJointOffset=<10 0 0>
     }]
    sphericalLimits [{limitRadius=50, drivingBone {boneName=spine_05}, sphereLocalOffset=<-30 25 8>}]
   }
   n9 = Anim Dynamics - Bone: dyn_hair_bk_l_01(ComponentPose=@n8, Alpha=@n4, ExternalForce="0,0,0") {
    angularDampingOverride=0.999
    angularSpringConstant=500
    bAngularSpring=true
    bGravityOverrideInSimSpace=true
    bOverrideAngularDamping=true
    bOverrideLinearDamping=true
    bUsePlanarLimit=false
    gravityScale=0.45
    linearDampingOverride=0.99
    alphaBoolBlend {blendOption=Linear}
    alphaScaleBiasClamp {interpSpeedDecreasing=2, interpSpeedIncreasing=4}
    boundBone {boneName=dyn_hair_bk_l_01}
    componentAppliedLinearAccClamp=<100 100 100>
    componentLinearVelScale=<50 50 50>
    physicsBodyDefinitions [{
      collisionType=CustomSphere
      sphereCollisionRadius=5
      boundBone {boneName=dyn_hair_bk_l_01}
      boxExtents=<5 30 30>
      constraintSetup {angularLimitsMax=<0 20 7>, angularLimitsMin=<0 -20 -65>, angularTarget=<0 0 0>}
      localJointOffset=<6.5 0 0>
     }]
    sphericalLimits [{limitRadius=5, drivingBone {boneName=neck_02}, sphereLocalOffset=<0 0 0>}]
   }
   n10 = Anim Dynamics - Bone: dyn_hair_bk_l_02(ComponentPose=@n9, Alpha=@n4, ExternalForce="0,0,0") {
    angularDampingOverride=0.999
    angularSpringConstant=500
    bAngularSpring=true
    bGravityOverrideInSimSpace=true
    bOverrideAngularDamping=true
    bOverrideLinearDamping=true
    bUseSphericalLimits=true
    gravityScale=0.7
    linearDampingOverride=0.99
    alphaBoolBlend {blendOption=Linear}
    alphaScaleBiasClamp {interpSpeedDecreasing=2, interpSpeedIncreasing=4}
    boundBone {boneName=dyn_hair_bk_l_02}
    componentAppliedLinearAccClamp=<100 100 100>
    componentLinearVelScale=<50 50 50>
    physicsBodyDefinitions [{
      collisionType=CustomSphere
      sphereCollisionRadius=4
      boundBone {boneName=dyn_hair_bk_l_02}
      boxExtents=<5 40 40>
      constraintSetup {angularLimitsMax=<0 20 7>, angularLimitsMin=<0 -20 -75>, angularTarget=<0 0 0>}
      localJointOffset=<10 0 0>
     }]
    sphericalLimits [{limitRadius=50, drivingBone {boneName=spine_05}, sphereLocalOffset=<-35 22 -8>}]
   }
   n11 = Anim Dynamics - Bone: dyn_hair_l_01(ComponentPose=@n10, Alpha=@n4, ExternalForce="0,0,0") {
    angularDampingOverride=0.999
    angularSpringConstant=500
    bAngularSpring=true
    bGravityOverrideInSimSpace=true
    bOverrideAngularDamping=true
    bOverrideLinearDamping=true
    bUsePlanarLimit=false
    gravityScale=0.45
    linearDampingOverride=0.99
    alphaBoolBlend {blendOption=Linear}
    alphaScaleBiasClamp {interpSpeedDecreasing=2, interpSpeedIncreasing=4}
    boundBone {boneName=dyn_hair_l_01}
    componentAppliedLinearAccClamp=<100 100 100>
    componentLinearVelScale=<50 50 50>
    physicsBodyDefinitions [{
      collisionType=CustomSphere
      sphereCollisionRadius=3
      boundBone {boneName=dyn_hair_l_01}
      boxExtents=<5 30 30>
      constraintSetup {angularLimitsMax=<0 8 3>, angularLimitsMin=<0 -8 -11>, angularTarget=<0 0 0>}
      localJointOffset=<-6.5 0 0>
     }]
    sphericalLimits [
     {limitRadius=5, drivingBone {boneName=neck_02}, sphereLocalOffset=<0 0 0>}
     {limitRadius=7, drivingBone {boneName=neck_01}, sphereLocalOffset=<5 10 0>}
    ]
   }
   n12 = Anim Dynamics - Bone: dyn_hair_l_02(ComponentPose=@n11, Alpha=@n4, ExternalForce="0,0,0") {
    angularDampingOverride=0.999
    angularSpringConstant=500
    bAngularSpring=true
    bGravityOverrideInSimSpace=true
    bOverrideAngularDamping=true
    bOverrideLinearDamping=true
    bUsePlanarLimit=false
    gravityScale=0.45
    linearDampingOverride=0.99
    alphaBoolBlend {blendOption=Linear}
    alphaScaleBiasClamp {interpSpeedDecreasing=2, interpSpeedIncreasing=4}
    boundBone {boneName=dyn_hair_l_02}
    componentAppliedLinearAccClamp=<100 100 100>
    componentLinearVelScale=<50 50 50>
    physicsBodyDefinitions [{
      collisionType=CustomSphere
      sphereCollisionRadius=2
      boundBone {boneName=dyn_hair_l_02}
      boxExtents=<5 30 30>
      constraintSetup {angularLimitsMax=<0 8 40>, angularLimitsMin=<0 -8 -15>, angularTarget=<0 0 0>}
      localJointOffset=<-6.5 0 0>
     }]
    planarLimits [{drivingBone {boneName=clavicle_l}, planeTransform {location=<0 -2 0>, rotation=<rot -90 90 0>}}]
    sphericalLimits [
     {limitRadius=5, drivingBone {boneName=neck_02}, sphereLocalOffset=<0 0 0>}
     {limitRadius=7, drivingBone {boneName=neck_01}, sphereLocalOffset=<5 10 0>}
    ]
   }
   n13 = Anim Dynamics - Bone: dyn_hair_l_03(ComponentPose=@n12, Alpha=@n4, ExternalForce="0,0,0") {
    angularDampingOverride=0.999
    angularSpringConstant=500
    bAngularSpring=true
    bGravityOverrideInSimSpace=true
    bOverrideAngularDamping=true
    bOverrideLinearDamping=true
    bUsePlanarLimit=false
    gravityScale=0.45
    linearDampingOverride=0.99
    alphaBoolBlend {blendOption=Linear}
    alphaScaleBiasClamp {interpSpeedDecreasing=2, interpSpeedIncreasing=4}
    boundBone {boneName=dyn_hair_l_03}
    componentAppliedLinearAccClamp=<100 100 100>
    componentLinearVelScale=<50 50 50>
    physicsBodyDefinitions [{
      collisionType=CustomSphere
      sphereCollisionRadius=3
      boundBone {boneName=dyn_hair_l_03}
      boxExtents=<5 30 30>
      constraintSetup {angularLimitsMax=<0 10 45>, angularLimitsMin=<0 -10 -25>, angularTarget=<0 0 0>}
      localJointOffset=<-6.5 0 0>
     }]
    sphericalLimits [
     {limitRadius=5, drivingBone {boneName=neck_02}, sphereLocalOffset=<0 0 0>}
     {limitRadius=7, drivingBone {boneName=neck_01}, sphereLocalOffset=<5 10 0>}
    ]
   }
   n14 = Anim Dynamics - Bone: dyn_hair_r_01(ComponentPose=@n13, Alpha=@n4, ExternalForce="0,0,0") {
    angularDampingOverride=0.999
    angularSpringConstant=1000
    bAngularSpring=true
    bGravityOverrideInSimSpace=true
    bOverrideAngularDamping=true
    bOverrideLinearDamping=true
    bUsePlanarLimit=false
    gravityScale=0.45
    linearDampingOverride=0.99
    alphaBoolBlend {blendOption=Linear}
    alphaScaleBiasClamp {interpSpeedDecreasing=2, interpSpeedIncreasing=4}
    boundBone {boneName=dyn_hair_r_01}
    componentAppliedLinearAccClamp=<100 100 100>
    componentLinearVelScale=<50 50 50>
    physicsBodyDefinitions [{
      collisionType=CustomSphere
      sphereCollisionRadius=3
      boundBone {boneName=dyn_hair_r_01}
      boxExtents=<5 30 30>
      constraintSetup {angularLimitsMax=<0 20 17>, angularLimitsMin=<0 -2.5 -8>, angularTarget=<0 0 0>}
      localJointOffset=<-6.5 0 0>
     }]
    sphericalLimits [
     {limitRadius=5, drivingBone {boneName=neck_02}, sphereLocalOffset=<0 0 0>}
     {limitRadius=7, drivingBone {boneName=neck_01}, sphereLocalOffset=<5 10 0>}
    ]
   }
   n15 = Anim Dynamics - Bone: dyn_hair_r_02(ComponentPose=@n14, Alpha=1, ExternalForce="0,0,0") {
    angularDampingOverride=0.999
    bGravityOverrideInSimSpace=true
    bOverrideAngularDamping=true
    bOverrideLinearDamping=true
    gravityScale=0.35
    linearDampingOverride=0.99
    alphaBoolBlend {blendOption=Linear}
    alphaScaleBiasClamp {interpSpeedDecreasing=2, interpSpeedIncreasing=4}
    boundBone {boneName=dyn_hair_r_02}
    componentAppliedLinearAccClamp=<100 100 100>
    componentLinearVelScale=<50 50 50>
    physicsBodyDefinitions [{
      sphereCollisionRadius=0
      boundBone {boneName=dyn_hair_r_02}
      boxExtents=<1 1 1>
      constraintSetup {angularLimitsMax=<0 20 20>, angularLimitsMin=<0 -5 -5>, angularTarget=<0 0 0>}
      localJointOffset=<0 0 0>
     }]
   }
   n16 = Anim Dynamics - Bone: dyn_hair_r_03(ComponentPose=@n15, Alpha=@n4, ExternalForce="0,0,0") {
    angularDampingOverride=0.999
    angularSpringConstant=1000
    bAngularSpring=true
    bGravityOverrideInSimSpace=true
    bOverrideAngularDamping=true
    bOverrideLinearDamping=true
    bUsePlanarLimit=false
    gravityScale=0.45
    linearDampingOverride=0.99
    alphaBoolBlend {blendOption=Linear}
    alphaScaleBiasClamp {interpSpeedDecreasing=2, interpSpeedIncreasing=4}
    boundBone {boneName=dyn_hair_r_03}
    componentAppliedLinearAccClamp=<100 100 100>
    componentLinearVelScale=<50 50 50>
    physicsBodyDefinitions [{
      collisionType=CustomSphere
      sphereCollisionRadius=3
      boundBone {boneName=dyn_hair_r_03}
      boxExtents=<5 30 30>
      constraintSetup {angularLimitsMax=<0 45 5>, angularLimitsMin=<0 -2 -30>, angularTarget=<0 0 0>}
      localJointOffset=<-6.5 0 0>
     }]
    sphericalLimits [{limitRadius=30, drivingBone {boneName=spine_05}, sphereLocalOffset=<5 7 -23.5>}]
   }
   n17 = "Transform (Modify) Bone - Bone: dyn_hair_r_03"(ComponentPose=@n16, Alpha=@n3, Translation="3.5,0,0", Rotation="40,0,0", Scale="1,1,1") {
    rotationMode=BMM_Additive
    translationMode=BMM_Additive
    alphaBoolBlend {blendOption=Linear}
    alphaScaleBiasClamp {bInterpResult=true, interpSpeedDecreasing=4, interpSpeedIncreasing=6}
    boneToModify {boneName=dyn_hair_r_03}
    rotation {pitch=40}
    translation {x=3.5}
   }
   n18 = Anim Dynamics - Bone: dyn_hair_fr_r_01(ComponentPose=@n17, Alpha=@n4, ExternalForce="0,0,0") {
    angularDampingOverride=0.9999
    angularSpringConstant=1000
    bAngularSpring=true
    bGravityOverrideInSimSpace=true
    bOverrideAngularDamping=true
    bOverrideLinearDamping=true
    bUsePlanarLimit=false
    gravityScale=0.25
    linearDampingOverride=0.9999
    alphaBoolBlend {blendOption=Linear}
    alphaScaleBiasClamp {interpSpeedDecreasing=2, interpSpeedIncreasing=2}
    boundBone {boneName=dyn_hair_fr_r_01}
    componentAppliedLinearAccClamp=<100 100 100>
    componentLinearVelScale=<50 50 50>
    physicsBodyDefinitions [{
      collisionType=CustomSphere
      sphereCollisionRadius=3
      boundBone {boneName=dyn_hair_fr_r_01}
      boxExtents=<5 30 30>
      constraintSetup {angularLimitsMax=<0 25 20>, angularLimitsMin=<0 -2 -30>, angularTarget=<0 0 0>}
      localJointOffset=<-6.5 0 0>
     }]
    sphericalLimits [
     {limitRadius=5, drivingBone {boneName=neck_02}, sphereLocalOffset=<0 0 0>}
     {limitRadius=7, drivingBone {boneName=neck_01}, sphereLocalOffset=<5 10 0>}
    ]
   }
   n19 = Anim Dynamics - Bone: dyn_hair_fr_r_02(ComponentPose=@n18, Alpha=@n4, ExternalForce="0,0,0") {
    angularDampingOverride=0.9999
    angularSpringConstant=1000
    bAngularSpring=true
    bGravityOverrideInSimSpace=true
    bOverrideAngularDamping=true
    bOverrideLinearDamping=true
    bUsePlanarLimit=false
    gravityScale=0.3
    linearDampingOverride=0.9999
    alphaBoolBlend {blendOption=Linear}
    alphaScaleBiasClamp {interpSpeedDecreasing=2, interpSpeedIncreasing=2}
    boundBone {boneName=dyn_hair_fr_r_02}
    componentAppliedLinearAccClamp=<100 100 100>
    componentLinearVelScale=<50 50 50>
    physicsBodyDefinitions [{
      collisionType=CustomSphere
      sphereCollisionRadius=3
      boundBone {boneName=dyn_hair_fr_r_02}
      boxExtents=<5 30 30>
      constraintSetup {angularLimitsMax=<0 20 30>, angularLimitsMin=<0 -5 -40>, angularTarget=<0 0 0>}
      localJointOffset=<-8 0 0>
     }]
    sphericalLimits [{limitRadius=30, drivingBone {boneName=spine_05}, sphereLocalOffset=<10 7 -24>}]
   }
   n20 = Anim Dynamics - Bone: dyn_hair_fl_01(ComponentPose=@n19, Alpha=@n4, ExternalForce="0,0,0") {
    angularDampingOverride=0.9999
    angularSpringConstant=1000
    bAngularSpring=true
    bGravityOverrideInSimSpace=true
    bOverrideAngularDamping=true
    bOverrideLinearDamping=true
    bUsePlanarLimit=false
    gravityScale=0.35
    linearDampingOverride=0.9999
    alphaBoolBlend {blendOption=Linear}
    alphaScaleBiasClamp {interpSpeedDecreasing=2, interpSpeedIncreasing=2}
    boundBone {boneName=dyn_hair_fl_01}
    componentAppliedLinearAccClamp=<100 100 100>
    componentLinearVelScale=<50 50 50>
    physicsBodyDefinitions [{
      collisionType=CustomSphere
      sphereCollisionRadius=3
      boundBone {boneName=dyn_hair_fl_01}
      boxExtents=<5 30 30>
      constraintSetup {angularLimitsMax=<1 3 8>, angularLimitsMin=<-1 -55 -20>, angularTarget=<0 0 0>}
      localJointOffset=<-15 0 0>
     }]
    sphericalLimits [
     {limitRadius=5, drivingBone {boneName=neck_02}, sphereLocalOffset=<0 0 0>}
     {limitRadius=7, drivingBone {boneName=neck_01}, sphereLocalOffset=<5 10 0>}
    ]
   }
   n21 = Anim Dynamics - Bone: dyn_hair_fl_02(ComponentPose=@n20, Alpha=@n4, ExternalForce="0,0,0") {
    angularDampingOverride=0.9999
    angularSpringConstant=1000
    bAngularSpring=true
    bGravityOverrideInSimSpace=true
    bOverrideAngularDamping=true
    bOverrideLinearDamping=true
    gravityScale=0.35
    linearDampingOverride=0.9999
    alphaBoolBlend {blendOption=Linear}
    alphaScaleBiasClamp {interpSpeedDecreasing=2, interpSpeedIncreasing=2}
    boundBone {boneName=dyn_hair_fl_02}
    componentAppliedLinearAccClamp=<100 100 100>
    componentLinearVelScale=<50 50 50>
    physicsBodyDefinitions [{
      collisionType=CustomSphere
      sphereCollisionRadius=1
      boundBone {boneName=dyn_hair_fl_02}
      boxExtents=<5 30 30>
      constraintSetup {angularLimitsMax=<1 5 12>, angularLimitsMin=<-1 -55 -30>, angularTarget=<0 0 0>}
      localJointOffset=<-11 0 0>
     }]
    planarLimits [{drivingBone {boneName=spine_05}, planeTransform {location=<0 0 0>, rotation=<rot -90 90 0>}}]
    sphericalLimits [{limitRadius=30, drivingBone {boneName=spine_05}, sphereLocalOffset=<10 7 -24>}]
   }
   n22 = Anim Dynamics - Bone: dyn_hair_fl_03(ComponentPose=@n21, Alpha=@n4, ExternalForce="0,0,0") {
    angularDampingOverride=0.9999
    angularSpringConstant=1000
    bAngularSpring=true
    bGravityOverrideInSimSpace=true
    bOverrideAngularDamping=true
    bOverrideLinearDamping=true
    gravityScale=0.35
    linearDampingOverride=0.9999
    alphaBoolBlend {blendOption=Linear}
    alphaScaleBiasClamp {interpSpeedDecreasing=2, interpSpeedIncreasing=2}
    boundBone {boneName=dyn_hair_fl_03}
    componentAppliedLinearAccClamp=<100 100 100>
    componentLinearVelScale=<50 50 50>
    physicsBodyDefinitions [{
      collisionType=CustomSphere
      sphereCollisionRadius=1
      boundBone {boneName=dyn_hair_fl_03}
      boxExtents=<5 30 30>
      constraintSetup {angularLimitsMax=<3 5 15>, angularLimitsMin=<-3 -55 -20>, angularTarget=<0 0 0>}
      localJointOffset=<-4 0 0>
     }]
    planarLimits [{drivingBone {boneName=spine_05}, planeTransform {location=<0 0 0>, rotation=<rot -90 90 0>}}]
    sphericalLimits [{limitRadius=30, drivingBone {boneName=spine_05}, sphereLocalOffset=<10 7 -24>}]
   }
   n23 = "Transform (Modify) Bone - Bone: dyn_hair_r_02"(ComponentPose=@n22, Alpha=1, Translation="0,0,0", Rotation=Get rot_hair_r() -> (rot_hair_r:Rotator), Scale="1,1,1") {
    rotationMode=BMM_Additive
    rotationSpace=BCS_BoneSpace
    alphaBoolBlend {blendOption=Linear}
    alphaScaleBiasClamp {bInterpResult=true, interpSpeedDecreasing=4, interpSpeedIncreasing=6}
    boneToModify {boneName=dyn_hair_r_02}
   }
   n24 = "Transform (Modify) Bone - Bone: dyn_hair_l_02"(ComponentPose=@n23, Alpha=1, Translation="0,0,0", Rotation=Get rot_hair_l() -> (rot_hair_l:Rotator), Scale="1,1,1") {
    rotationMode=BMM_Additive
    rotationSpace=BCS_BoneSpace
    alphaBoolBlend {blendOption=Linear}
    alphaScaleBiasClamp {bInterpResult=true, interpSpeedDecreasing=4, interpSpeedIncreasing=6}
    boneToModify {boneName=dyn_hair_l_02}
   }
   n25 = "Transform (Modify) Bone - Bone: dyn_hair_bk_01"(ComponentPose=@n24, Alpha=1, Translation="-0.255,-0.950,1.793", Rotation="-10.870,-24.629,10.319", Scale="1,1,1") {
    rotationMode=BMM_Additive
    rotationSpace=BCS_BoneSpace
    translationMode=BMM_Additive
    translationSpace=BCS_BoneSpace
    alphaScaleBiasClamp {bInterpResult=true}
    boneToModify {boneName=dyn_hair_bk_01}
    rotation=<rot -10.87 -24.629 10.319>
    translation=<-0.255 -0.95 1.793>
   }
   n26 = "Transform (Modify) Bone - Bone: dyn_hair_bk_l_01"(ComponentPose=@n25, Alpha=1, Translation="-2.200,-1.829,0.130", Rotation="-13.692,-124.669,36.035", Scale="1,1,1") {
    rotationMode=BMM_Additive
    rotationSpace=BCS_BoneSpace
    translationMode=BMM_Additive
    translationSpace=BCS_BoneSpace
    alphaScaleBiasClamp {bInterpResult=true}
    boneToModify {boneName=dyn_hair_bk_l_01}
    rotation=<rot -13.692 -124.669 36.035>
    translation=<-2.2 -1.829 0.13>
   }
   n27 = "Transform (Modify) Bone - Bone: dyn_hair_bk_l_02"(ComponentPose=@n26, Alpha=1, Translation="-2,0,0", Rotation="-22.988,71.518,14.978", Scale="1,1,1") {
    rotationMode=BMM_Additive
    rotationSpace=BCS_BoneSpace
    translationMode=BMM_Additive
    translationSpace=BCS_BoneSpace
    alphaScaleBiasClamp {bInterpResult=true}
    boneToModify {boneName=dyn_hair_bk_l_02}
    rotation=<rot -22.988 71.518 14.978>
    translation {x=-2}
   }
   n28 = "Transform (Modify) Bone - Bone: dyn_hair_bk_r_01"(ComponentPose=@n27, Alpha=1, Translation="0,-5,0", Rotation="-0.000,-41.701,0.000", Scale="1,1,1") {
    rotationMode=BMM_Additive
    rotationSpace=BCS_BoneSpace
    translationMode=BMM_Additive
    translationSpace=BCS_BoneSpace
    alphaScaleBiasClamp {bInterpResult=true}
    boneToModify {boneName=dyn_hair_bk_r_01}
    rotation=<rot 0 -41.701 0>
    translation {y=-5}
   }
   n29 = "Transform (Modify) Bone - Bone: dyn_hair_r_01"(ComponentPose=@n28, Alpha=1, Translation="0,0,0", Rotation="-57.745,42.416,-74.465", Scale="1,1,1") {
    rotationMode=BMM_Additive
    rotationSpace=BCS_BoneSpace
    translationMode=BMM_Additive
    translationSpace=BCS_BoneSpace
    alphaScaleBiasClamp {bInterpResult=true}
    boneToModify {boneName=dyn_hair_r_01}
    rotation=<rot -57.745 42.416 -74.465>
   }
   n30 = "Transform (Modify) Bone - Bone: dyn_hair_l_01"(ComponentPose=@n29, Alpha=1, Translation="0,0,0", Rotation="23.545,92.363,55.304", Scale="1,1,1") {
    rotationMode=BMM_Additive
    rotationSpace=BCS_BoneSpace
    translationMode=BMM_Additive
    translationSpace=BCS_BoneSpace
    alphaScaleBiasClamp {bInterpResult=true}
    boneToModify {boneName=dyn_hair_l_01}
    rotation=<rot 23.545 92.363 55.304>
   }
   n31 = "Transform (Modify) Bone - Bone: dyn_hair_bk_01"(ComponentPose=@n30, Alpha=1, Translation="0,0,0", Rotation="-0.430,4.503,-1.610", Scale="1,1,1") {
    rotationMode=BMM_Additive
    rotationSpace=BCS_BoneSpace
    translationMode=BMM_Additive
    translationSpace=BCS_BoneSpace
    alphaScaleBiasClamp {bInterpResult=true}
    boneToModify {boneName=dyn_hair_bk_01}
    rotation=<rot -0.43 4.503 -1.61>
   }
   n32 = "Transform (Modify) Bone - Bone: dyn_hair_bk_l_01"(ComponentPose=@n31, Alpha=1, Translation="0,0,0", Rotation="-35.034,-26.549,-0.000", Scale="1,1,1") {
    rotationMode=BMM_Additive
    rotationSpace=BCS_BoneSpace
    translationMode=BMM_Additive
    translationSpace=BCS_BoneSpace
    alphaScaleBiasClamp {bInterpResult=true}
    boneToModify {boneName=dyn_hair_bk_l_01}
    rotation=<rot -35.034 -26.549 0>
   }
   n33 = "Transform (Modify) Bone - Bone: dyn_hair_bk_r_01"(ComponentPose=@n32, Alpha=1, Translation="0,0,0", Rotation="-19.414,-29.532,-1.685", Scale="1,1,1") {
    rotationMode=BMM_Additive
    rotationSpace=BCS_BoneSpace
    translationMode=BMM_Additive
    translationSpace=BCS_BoneSpace
    alphaScaleBiasClamp {bInterpResult=true}
    boneToModify {boneName=dyn_hair_bk_r_01}
    rotation=<rot -19.414 -29.532 -1.685>
   }
   Output Pose(Result=Component To Local(ComponentPose=@n33)) {layerGroup=None}
   AnimGraphNode_SubInput_0 = Input Pose()
  ]}
 EventGraph {nodes [
   Event BlueprintUpdateAnimation()
   > n1 = Sequence!(execute) -> (then_0, then_1, then_2:Exec)
   n2 = GetPartSkeletalMeshComponent()
   n3 = Get Distance Between Two Sockets!(
    Component:Skeletal Mesh Component Object Reference=@n2
    SocketOrBoneNameA:Name=dyn_collar_r
    SocketSpaceA:ERelativeTransformSpace Enum=RTS_Component
    SocketOrBoneNameB:Name=dyn_hair_r_03
    SocketSpaceB:ERelativeTransformSpace Enum=RTS_Component
    bRemapRange:bool=true
    InRangeMin:float=-10
    InRangeMax:float=10
    OutRangeMin:float=0
    OutRangeMax:float=1
   ) -> (ReturnValue:float) {is=title K2Node_CallFunction}
   > Set hair_dist(hair_dist:double=float - float(A=@n3, B=1) {is=title K2Node_CallFunction}) -> (Output_Get:double)
   > Set rot_hair_r(rot_hair_r:Rotator=MakeRotator(Pitch=float * float(A=@n3, B=-5))) -> (Output_Get:Rotator)
   n4 = float - float(
    A=Get Distance Between Two Sockets(
     Component=@n2
     SocketOrBoneNameA=dyn_collar_l
     SocketSpaceA=RTS_Component
     SocketOrBoneNameB=dyn_hair_l_03
     SocketSpaceB=RTS_Component
     bRemapRange=true
     InRangeMin=0
     InRangeMax=100
     OutRangeMin=0
     OutRangeMax=1
    )
    B=1
   )
   > Set rot_hair_l(rot_hair_l:Rotator=MakeRotator(Yaw=float * float(A=@n4, B=5))) -> (Output_Get:Rotator)
   > n5 = Branch(Condition=Get bIsTargeting())
   > Set target_hair_r(target_hair_r:double=1) -> (Output_Get:double)
   Set target_hair_r(execute=@n5.else, target_hair_r=0)
   n6 = Branch(execute=@n1.then_1, Condition=Get bIsSkydiveFloating())
   > Set Skydiving_Floating(Skydiving_Floating:double=1) -> (Output_Get:double)
   Set Skydiving_Floating(execute=@n6.else, Skydiving_Floating=0)
   n7 = Branch(execute=@n1.then_2, Condition=Get bIsParachuteOpened())
   > Set ParachuteOpen(ParachuteOpen:double=1) -> (Output_Get:double)
   Set ParachuteOpen(execute=@n7.else, ParachuteOpen=0)
  ]}
]
variables [
 hair_alpha {type=number, default=0.875}
 hair_dist {type=number}
 ParachuteOpen {type=number}
 rot_hair_l {type=Rotator}
 rot_hair_r {type=Rotator}
 Skydiving_Floating {type=number}
 target_hair_r {type=number}
]
